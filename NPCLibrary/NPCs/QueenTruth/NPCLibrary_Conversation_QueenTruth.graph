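# Conversation database for the QueenTruth NPC, giving it Simpsons-themed greetings and a hub into shared NPC conversations.
# One entry point leads to a Speech node that picks lines by priority, based on first meeting, conversation count or player outfit.
# It branches through link nodes to library quests, basic services and two Season 38 story quests.
# A further Season 38 LAF combat link requires that no LAF quest is active.
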ConversationDatabase /NPCLibrary/NPCs/QueenTruth/NPCLibrary_Conversation_QueenTruth
entry_points [{tag=Conversation.Entry.QueenTruth, destinations [#1]}]
nodes [
 ConversationEntryPointNode_7 {
  id=#1
  class=ConversationEntryPointNode
  outputs [#2]
  properties {entryTag {tagName=Conversation.Entry.QueenTruth}}
 }
 Speech_C_4 {
  id=#2
  class=Speech_C
  outputs [
   #3
   #4
   #5
   #6
   #7
  ]
  properties {
   nodeName=""
   generalConfig {
    defaultMessage="You didn't see me, I was never here."
    contextualMessages [
     {
      message="You're lucky, my bus just takes me to school."
      contextRequirements [{requirement {class=FortControllerRequirement_IsFirstEverConversationWithNPC}}]
      priority {curve {rowName=Tandem.Speech.Priority.FirstEverInteraction}}
     }
     {
      message="Don't have a cow, man."
      contextRequirements [{requirement {class=FortControllerRequirement_CheckNPCMatchConversationCount, properties {nPCUniqueIdTag {tagName=AISpawnerData.Type.Tandem.QueenTruth}}}}]
      priority {curve {rowName=None}}
     }
     {
      message="You didn't see me, I was never here."
      contextRequirements [{requirement {class=FortControllerRequirement_CheckNPCMatchConversationCount, properties {nPCUniqueIdTag {tagName=AISpawnerData.Type.Tandem.QueenTruth}}}}]
      priority {curve {rowName=None}}
     }
     {
      message="Way to destroy the town, Homer."
      contextRequirements [{requirement {class=FortControllerRequirement_HasCID, properties {allowedCIds [/BRCosmetics/Athena/Items/Cosmetics/Characters/Character_SilentNovel_Vane]}}}]
      priority {value=2, curve {rowName=Tandem.Speech.Priority.OutfitReaction}}
     }
     {
      message="I can't believe I'm saying this, but I think dad really needs our help this time."
      contextRequirements [{requirement {class=FortControllerRequirement_HasCID, properties {allowedCIds [/BRCosmetics/Athena/Items/Cosmetics/Characters/Character_FruitFire_Nudge]}}}]
      priority {value=2, curve {rowName=Tandem.Speech.Priority.OutfitReaction}}
     }
     {
      message=How can someone as lazy as Homer cause this much trouble?
      contextRequirements [{requirement {class=FortControllerRequirement_HasCID, properties {allowedCIds [/BRCosmetics/Athena/Items/Cosmetics/Characters/Character_FrostGalore]}}}]
      priority {value=2, curve {rowName=Tandem.Speech.Priority.OutfitReaction}}
     }
     {
      message=How many of you are there?
      contextRequirements [{requirement {
         class=FortControllerRequirement_HasCID
         properties {allowedCIds [
           &Character_UndergroundRebel_Fashion
           /Game/Athena/Items/Cosmetics/Characters/CID_005_Athena_Commando_M_Default
           /BRCosmetics/Athena/Items/Cosmetics/Characters/CID_513_Athena_Commando_M_CubePaintJonesy
           /BRCosmetics/Athena/Items/Cosmetics/Characters/CID_010_Athena_Commando_M
           /BRCosmetics/Athena/Items/Cosmetics/Characters/CID_025_Athena_Commando_M
           /BRCosmetics/Athena/Items/Cosmetics/Characters/CID_018_Athena_Commando_M
           /BRCosmetics/Athena/Items/Cosmetics/Characters/CID_051_Athena_Commando_M_HolidayElf
           /BRCosmetics/Athena/Items/Cosmetics/Characters/CID_042_Athena_Commando_M_Cyberpunk
           /BRCosmetics/Athena/Items/Cosmetics/Characters/CID_080_Athena_Commando_M_Space
           /BRCosmetics/Athena/Items/Cosmetics/Characters/CID_085_Athena_Commando_M_Twitch
           /BRCosmetics/Athena/Items/Cosmetics/Characters/CID_072_Athena_Commando_M_Scout
           /BRCosmetics/Athena/Items/Cosmetics/Characters/CID_093_Athena_Commando_M_Dinosaur
           /BRCosmetics/Athena/Items/Cosmetics/Characters/CID_095_Athena_Commando_M_Founder
           /BRCosmetics/Athena/Items/Cosmetics/Characters/CID_168_Athena_Commando_M_Shark
           /BRCosmetics/Athena/Items/Cosmetics/Characters/CID_176_Athena_Commando_M_Lifeguard
           /BRCosmetics/Athena/Items/Cosmetics/Characters/CID_180_Athena_Commando_M_Scuba
           /BRCosmetics/Athena/Items/Cosmetics/Characters/CID_184_Athena_Commando_M_DurrburgerWorker
           /BRCosmetics/Athena/Items/Cosmetics/Characters/CID_331_Athena_Commando_M_Taxi
           /BRCosmetics/Athena/Items/Cosmetics/Characters/CID_548_Athena_Commando_M_YellowCamoA
           /Game/Athena/Items/Cosmetics/Characters/CID_560_Athena_Commando_M_RebirthDefaultA
           /BRCosmetics/Athena/Items/Cosmetics/Characters/CID_568_Athena_Commando_M_RebirthSoldier
           /BRCosmetics/Athena/Items/Cosmetics/Characters/CID_756_Athena_Commando_M_JonesyAgent
           /BRCosmetics/Athena/Items/Cosmetics/Characters/CID_774_Athena_Commando_M_TacticalScuba
           /BRCosmetics/Athena/Items/Cosmetics/Characters/CID_798_Athena_Commando_M_JonesyVagabond
           /BRCosmetics/Athena/Items/Cosmetics/Characters/CID_836_Athena_Commando_M_JonesyFlare
           &CID_883_Athena_Commando_M_ChOneJonesy
           /BRCosmetics/Athena/Items/Cosmetics/Characters/CID_981_Athena_Commando_M_JonesyHoliday
           &CID_989_Athena_Commando_M_ProgressiveJonesy
           /BRCosmetics/Athena/Items/Cosmetics/Characters/CID_A_406_Athena_Commando_M_RebirthFresh
           /BRCosmetics/Athena/Items/Cosmetics/Characters/Character_JonesyOrangeFNCS
           /BRCosmetics/Athena/Items/Cosmetics/Characters/CID_A_431_Athena_Commando_M_JonesyOrange
           /BRCosmetics/Athena/Items/Cosmetics/Characters/CID_A_446_Athena_Commando_M_PennantSeasOne
           /BRCosmetics/Athena/Items/Cosmetics/Characters/Character_TacticKale
           /BRCosmetics/Athena/Items/Cosmetics/Characters/CID_A_313_Athena_Commando_M_BlizzardBomber
          ]}
        }}]
      priority {value=2, curve {rowName=Tandem.Speech.Priority.OutfitReaction}}
     }
     {
      message="I guess... do... have a cow, man?"
      contextRequirements [{requirement {
         class=FortControllerRequirement_HasCID
         properties {allowedCIds [/BRCosmetics/Athena/Items/Cosmetics/Characters/CID_A_118_Athena_Commando_M_JonesyCattle]}
        }}]
      priority {value=2, curve {rowName=Tandem.Speech.Priority.OutfitReaction}}
     }
    ]
   }
  }
 }
 ConversationLinkNode_4 {
  id=#7
  class=ConversationLinkNode
  properties {remoteEntryTag {tagName=Conversation.Entry.Season38.LAF.NPC.Combat}}
  sub_nodes [FortConversationRequirement_ControllerRequirement_0 {
    class=FortConversationRequirement_ControllerRequirement
    properties {
     nodeName="Require: Has not started a L&&F Quest"
     requirement {class=FortControllerRequirement_HasOwnedTag, properties {bInverseCondition=true, ownedTag {tagName=Athena.Quests.S38.LAF.QuestActive}}}
    }
   }]
 }
 ConversationLinkNode_3 {
  id=#6
  class=ConversationLinkNode
  properties {remoteEntryTag {tagName=Conversation.Entry.Season38.Story.P03.Q07.01}}
 }
 ConversationLinkNode_2 {
  id=#5
  class=ConversationLinkNode
  properties {remoteEntryTag {tagName=Conversation.Entry.Season38.Story.P03.Q05}}
 }
 ConversationLinkNode_0 {id=#4, class=ConversationLinkNode, properties {remoteEntryTag {tagName=Conversation.Entry.BasicServices}}}
 ConversationLinkNode_1 {id=#3, class=ConversationLinkNode, properties {remoteEntryTag {tagName=Conversation.Entry.NPCLibrary.Quests}}}
]
